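# Animation Blueprint for the Mystic Book backpack cosmetic that makes its dyn_book bone hover and sway.
# Each update builds a Floater vector from two time-driven sine waves on X and Z, applied as an additive bone offset.
# That feeds spring-based Anim Dynamics, then an additive upward offset.
# dynAlpha weights those last two stages, dropping sharply when the cosmetic is displayed in the frontend and staying full otherwise.
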AnimBlueprint /BRCosmetics/Accessories/FORT_Backpacks/Backpack_M_MED_Mystic_Book/Meshes/M_MED_Mystic_Book_Pack_AnimBP
target_skeleton=&/M_MED_Mystic_Book_Pack_Skeleton
graphs [
 AnimGraph {nodes [
   n1 = Local Space Ref Pose()
   n2 = Get dynAlpha() -> (dynAlpha:double)
   n3 = Anim Dynamics - Bone: dyn_book(
    ComponentPose="Transform (Modify) Bone - Bone: dyn_book"(ComponentPose=Local To Component(LocalPose=@n1), Alpha=1, Translation=Get Floater() -> (Floater:Vector), Rotation="0,0,0", Scale="1,1,1") {translationMode=BMM_Additive, alphaBoolBlend {blendOption=Linear}, boneToModify {boneName=dyn_book}}
    Alpha=@n2
    ExternalForce="0,0,0"
   ) {
    bAngularSpring=true
    bGravityOverrideInSimSpace=true
    bLinearSpring=true
    bOverrideLinearDamping=true
    bUseGravityOverride=true
    bUseSphericalLimits=true
    linearDampingOverride=0.988
    linearSpringConstant=45
    alphaBoolBlend {blendOption=Linear}
    boundBone {boneName=dyn_book}
    componentAppliedLinearAccClamp=<150000 10000 75000>
    componentLinearAccScale {x=0.04, z=0.05}
    componentLinearVelScale {x=0.04, z=0.05}
    gravityOverride {z=-980}
    physicsBodyDefinitions [{
      boundBone {boneName=dyn_book}
      boxExtents=<60 60 60>
      constraintSetup {
       linearXLimitType=Free
       linearYLimitType=Free
       linearZLimitType=Free
       angularLimitsMax=<0 0 2>
       angularLimitsMin=<0 0 -2>
       angularTarget=<0 0 0>
       linearAxesMax=<2 2 2>
       linearAxesMin=<-2 0 -2>
      }
      localJointOffset=<0 0 0>
     }]
   }
   n4 = "Transform (Modify) Bone - Bone: dyn_book"(ComponentPose=@n3, Alpha=@n2, Translation="0.000,0.000,14", Rotation="0,0,0", Scale="1,1,1") {translationMode=BMM_Additive, alphaBoolBlend {blendOption=Linear}, boneToModify {boneName=dyn_book}, translation {z=14}}
   Output Pose(Result=Component To Local(ComponentPose=@n4)) {layerGroup=None}
  ]}
 EventGraph {nodes [
   Event BlueprintUpdateAnimation()
   n1 = GetGameTimeInSeconds()
   n2 = "Sin (Radians)"(A=float * float(A=float * float(A=0.9, B=3), B=@n1))
   n3 = "Sin (Radians)"(A=float * float(A=float * float(A=0.8, B=-1), B=@n1))
   > Set Floater(Floater:Vector=MakeVector(X=float * float(A=@n3, B=1.25), Y=0, Z=float * float(A=@n2, B=1.25))) -> (Output_Get:Vector)
   > n4 = Branch(Condition=IsCosmeticObjectDisplayedInFrontendContext!(CosmeticObject:Object Reference) -> (ReturnValue:bool))
   > Set dynAlpha(dynAlpha:double=0.175) -> (Output_Get:double)
   Set dynAlpha(execute=@n4.else, dynAlpha=1)
   TryGetPawnOwner()
  ]}
]
variables [
 dynAlpha {type=number}
 Floater {type=Vector}
]
